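# Niagara system for the Helios Hades environment that scatters drifting, fading skull particles over the world landscape, using one GPU emitter with two mesh renderers and a sprite renderer.
# Particles spawn in a cylinder, sample landscape height, then pass through water spawn-mask textures and a kill volume at a fixed world area.
# Update re-samples terrain, drives horizontal motion with curl noise, wind and distance-field avoidance, and fades color and alpha through curves and a region mask.
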NiagaraSystem /FNGameplayCues/Environments/Helios/FX/Hades/Effects/NS_Hades_Camera_Skulls_OnWorld
dependencies {
 usedDataInterfaces [/Script/Niagara.NiagaraDataInterfaceTexture, &NiagaraDataInterfaceColorCurve, &NiagaraDataInterfaceCurve]
 usedDynamicInputs []
 usedModules [
  &SystemState
  &self:SystemSpawnScript.SystemScriptSource.SystemScriptGraph.NiagaraNodeAssignment_0.SetVariables_#1
  &self:SystemSpawnScript.SystemScriptSource.SystemScriptGraph.NiagaraNodeAssignment_1.SetVariables_#2
 ]
}
emitters {Skulls_Landscape {
  simTarget=GPUComputeSim
  emitterUpdateScript [
   &EmitterState()
   &SpawnRate(SpawnRate=250, Use Spawn Probability=false)
  ]
  particleSpawnScript [
   &AddVelocity(Velocity=<0 50 0>)
   &InitializeParticle(
    Color Maximum=$System.Color2
    Color Minimum=$System.Color1
    Color Mode=::NewEnumerator2
    Lifetime Max=6
    Lifetime Min=4
    Lifetime Mode=::NewEnumerator1
    Mesh Scale Mode=::NewEnumerator1
    Mesh Uniform Scale Max=10
    Mesh Uniform Scale Min=5
    Sprite Size Mode=::NewEnumerator1
    Sprite UV Mode=::NewEnumerator1
    Uniform Sprite Size Max=125
    Uniform Sprite Size Min=80
    UsePositionOffset=false
    Write Material Random=true
   )
   /SetVariables(Particles.Terrain Offset=25) {isSetParameters=true}
   &ShapeLocation(
    Cylinder Radius=5000
    MaskSpawnGroup=false
    Override Seed=false
    Shape Primitive=::NewEnumerator1
    Surface Only=true
    Surface Only Band Thickness=0.8
    Use Endcaps In Surface Only Mode=false
    UseOwnerScale=true
   )
   &self:Sample_Landscape_01() {module=Sample_Landscape_01}
   /SetVariables(Particles.New Vector 3f:Vector3f=<0 0 1>) {isSetParameters=true}
   &DynamicMaterialParameters(
    Index 0 Param 1=&RandomRangeFloat(Minimum=50, Maximum=100)
    Index 0 Param 2=5
    Index 0 Param 3=15
    Index 0 Param 4=&RandomRangeFloat(Minimum=0, Maximum=0)
    Index 1 Param 1=50
    Index 2 Param 1=&RandomRangeFloat(Minimum=-0.1, Maximum=-0.1)
    Index 2 Param 2=&RandomRangeFloat(Minimum=0.1, Maximum=0)
    Index 2 Param 3=&RandomRangeFloat(Minimum=0, Maximum=0)
    Index 2 Param 4=&RandomRangeFloat(Minimum=0.2, Maximum=0.3)
    Module.Index0Param0Write=true
    Module.Index0Param1Write=true
    Module.Index0Param2Write=true
    Module.Index0Param3Write=true
    Module.Index1Param0Write=true
    Module.Index1Param1Write=true
    Module.Index1Param2Write=true
    Module.Index1Param3Write=true
    Module.Index2Param0Write=true
    Module.Index2Param1Write=true
    Module.Index2Param2Write=true
    Module.Index2Param3Write=true
    Write Parameter Index 1=true
    Write Parameter Index 2=true
   )
   /SetVariables(Particles.Linear Color=<rgba 1 0 0 1>, Particles.WaveScale:Vector3f=<6 6 6>) {isSetParameters=true}
   &self:ScratchModule_08(
    BoxCenter:Vector3f=<32429 -64990 940>
    BoxExtents:Vector3f=<130000 130000 1005>
    TextureSample=DataInterface() {propertyValues {
      Texture=/Game/Environments/Helios/FX/Hades/Textures/T_Hades_WaterSpawnMask_Position_01
      TextureUserParameter {parameter {name=None, typeDefHandle {registeredTypeIndex=2}, varData []}}
     }}
   ) {module=ScratchModule_08}
   /SetVariables(Particles.Position=&Add_Vector(
     A=Linked() {linkedVariable {name=Particles.Position, type=NiagaraPosition}}
     B=&MakeVector(X=0, Y=0, Z=$Particles.Terrain Offset)
    )) {isSetParameters=true}
   &KillParticlesInVolume(
    Box Size:Vector3f=<15000 15000 250000>
    Kill Shape=::NewEnumerator1
    Volume Origin:Vector3f=&AddVectorToPosition(Position=<49038 -89768 3140>, Vector=<0 0 0>)
   )
   &self:RainTextureMasking_01(
    AlphaThreshold=0.5
    BoxCenter:NiagaraPosition=<32429 -64990 3473.7344>
    BoxExtents:Vector3f=<130000 130000 1>
    Position:NiagaraPosition=$Particles.Position
    RejectionTexture=DataInterface() {propertyValues {
      Texture=/Game/Environments/Helios/FX/Hades/Textures/T_Hades_WaterSpawnMask_01
      TextureUserParameter {parameter {name=None, typeDefHandle {registeredTypeIndex=2}, varData []}}
     }}
   ) {module=RainTextureMasking_01}
   &ApplyInitialForces()
  ]
  particleUpdateScript [
   &self:Sample_Landscape_UPdate_01(Bias:NiagaraFloat=$Particles.Terrain Offset, New Landscape Sample:NiagaraDataInterfaceLandscape=$User.Landscape) {module=Sample_Landscape_UPdate_01}
   &ParticleState()
   &CurlNoiseForce(
    Curl Noise Cone Mask Angle=25
    Mask Curl Noise=true
    Noise Frequency=&RandomRangeFloat(Minimum=2, Maximum=4)
    Noise Strength=&Multiply_Float(A=&RandomRangeFloat(Minimum=25, Maximum=100), B=1)
    Pan Noise=true
   )
   &WindForce(Use Depth Buffer=false, Use Landscape=false, Wind Speed=<0 25 0>)
   &Drag(Drag=0.1, Use Linear Drag=true, Use Rotational Drag=true)
   &SolveForcesAndVelocity(
    Acceleration Limit=5
    Clamp Velocity=true
    Force:Vector3f=&Multiply_Vector(A=$Transient.PhysicsForce, B=<1 1 0>)
    Limit Acceleration=true
    Speed Limit=25
    Velocity:Vector3f=&Multiply_Vector(A=$Particles.Velocity, B=<1 1 0>)
   )
   &CalculateAccurateVelocity() {enabled=false}
   &CurlNoiseForce(
    Curl Noise Cone Mask Angle=80
    Curl Noise Cone Mask Falloff Angle=90
    Mask Curl Noise=true
    Noise Frequency=&RandomRangeFloat(Minimum=0.25, Maximum=1)
    Noise Strength=&Multiply_Float(A=&RandomRangeFloat(Minimum=35, Maximum=100), B=1)
    Pan Noise=false
    Pan Noise Field=<0 0 2>
   ) {module=CurlNoiseForce001}
   /Niagara/Modules/Math/AvoidDistanceFieldSurfaces_GPU(
    Avoid Nearest Surface=true
    Avoid Oncoming Surfaces=true
    Falloff Distance=1
    Max Trace Distance=1
    Nearest Surface Avoidance Strength=1
    Obstactle Avoidance Strength=1
    Trace Vector:Vector3f=&Multiply_Vector(A=$Particles.Velocity, B=<1 1 0>)
   )
   &SolveForcesAndVelocity(
    Acceleration Limit=5
    Clamp Velocity=true
    Force:Vector3f=&Multiply_Vector(A=$Transient.PhysicsForce, B=<1 1 0>)
    Limit Acceleration=true
    Speed Limit=100
    Velocity:Vector3f=&Multiply_Vector(A=$Particles.Velocity, B=<1 1 0>)
   ) {module=SolveForcesAndVelocity001}
   &RegionMask(
    Invert Region=true
    Offset:Vector3f=<49038 -89768 3140>
    Offset Space:ENiagaraCoordinateSpace=::Local
    Region Dimensions:Vector3f=<15000 15000 15000>
    Region Type:ENiagaraDebugDrawShape=::NewEnumerator1
    UseOffset=true
    UseRotation=false
   )
   &ScaleColor(
    Scale Alpha=&FloatFromCurve(
     FloatCurve=DataInterface() {propertyValues {Curve {keys [
         {interpMode=RCIM_Cubic, tangentMode=RCTM_User, time=0, value=0}
         {interpMode=RCIM_Cubic, tangentMode=RCTM_User, time=0.5, value=1}
         {tangentMode=RCTM_User, time=1, value=0}
        ]}}}
     Scale Curve=&Multiply_Float(A=0.5, B=$StackContext.RegionMask.Float)
    )
    Scale RGB=&Multiply_VectorByFloat(Vector=<1 1 1>, Float=&RandomRangeFloat(Minimum=1, Maximum=2))
    ScaleA=true
    ScaleRGB=true
   ) {module=ScaleColor001}
   /SetVariables(
    Particles.Linear Color=&MultiplyLinearColor(
     Linear Color=$Particles.Color
     Scale Factor=&MakeVector4_FromFloat(Float=&FloatFromCurve(
       FloatCurve=DataInterface() {propertyValues {Curve {keys [
           {interpMode=RCIM_Cubic, tangentMode=RCTM_User, time=0, value=0}
           {interpMode=RCIM_Cubic, tangentMode=RCTM_User, time=0.5, value=1}
           {tangentMode=RCTM_User, time=1, value=0}
          ]}}}
       Scale Curve=$System.ColorIntensityMult
      ))
    )
    Particles.Linear Color001:LinearColor=&MultiplyLinearColor(
     Linear Color=&ColorFromCurve(ColorCurve=DataInterface() {propertyValues {
        AlphaCurve {keys [<at 0 0 RCIM_Cubic>,<at 0.1218 1>,<at 1 0 RCIM_Cubic>]}
        BlueCurve {keys [<at 0 0.0688 RCIM_Cubic>,<at 1 1 RCIM_Cubic>]}
        GreenCurve {keys [<at 0 1 RCIM_Cubic>,<at 1 1 RCIM_Cubic>]}
        RedCurve {keys [<at 0 0 RCIM_Cubic>,<at 1 1 RCIM_Cubic>]}
       }})
     Scale Factor=&MakeVector4_FromFloat(Float=&FloatFromCurve(
       FloatCurve=DataInterface() {propertyValues {Curve {keys [<at 0 1>,<at 1 0>]}}}
       Scale Curve=&Multiply_Float(A=$System.ColorIntensityMult, B=4)
      ))
    )
   ) {isSetParameters=true}
   &CameraOffset(Camera Offset Amount=10) {module=CameraOffset001}
  ]
  rendererClasses [&NiagaraMeshRendererProperties, &NiagaraSpriteRendererProperties]
  renderers [
   {rendererClass=&NiagaraMeshRendererProperties, rendererIndex=0}
   {rendererClass=&NiagaraMeshRendererProperties, rendererIndex=1}
   {rendererClass=&NiagaraSpriteRendererProperties, rendererIndex=2}
  ]
 }}
summary {
 systemName=NS_Hades_Camera_Skulls_OnWorld
 emitters [{
   emitterName=Skulls_Landscape
   simTarget=GPUComputeSim
   rendererClasses [&NiagaraMeshRendererProperties, &NiagaraSpriteRendererProperties]
  }]
 userVariables [
  User.Landscape {
   type=NiagaraDataInterfaceLandscape
   description=""
   defaultValue {
    struct=NiagaraExt_VariableValue_DataInterface
    value {
     dataInterface=&self:NiagaraDataInterfaceLandscape_0
     dataInterfaceClass=&NiagaraDataInterfaceLandscape
    }
   }
  }
  User.MaxVel {type=NiagaraFloat, defaultValue=100, description=""}
  User.RadiusSpread {type=NiagaraFloat, defaultValue=1500, description=""}
  User.WaterBodyIndex {type=NiagaraFloat, defaultValue=0, description=""}
  User.WaterDepth {type=NiagaraFloat, defaultValue=0, description=""}
  User.WaterPlaneLocation {type=Vector3f, description="", defaultValue {struct=Vector3f, value=<0 0 0>}}
  User.WaterPlaneNormal {type=Vector3f, description="", defaultValue {struct=Vector3f, value=<0 0 0>}}
  User.WaveAttenuationFactor {type=NiagaraFloat, defaultValue=0, description=""}
  User.WaveTime {type=NiagaraFloat, defaultValue=0, description=""}
 ]
}
