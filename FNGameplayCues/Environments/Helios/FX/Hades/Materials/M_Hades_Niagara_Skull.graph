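# Translucent unlit material for Niagara skull mesh particles in the Helios Hades effects, with a vertex-blue mask separating skull regions.
# Radially panned normals distort noise that drives caustics for an animated pattern, alongside a Fresnel-shaded color with per-instance hue shift.
# Opacity combines inverted Fresnel, depth fade, particle alpha and synced distance fade.
# Dynamic parameters let particles control depth fade, noise motion and a world-space curl-noise vector.
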Material /FNGameplayCues/Environments/Helios/FX/Hades/Materials/M_Hades_Niagara_Skull
graph {
 properties {
  bUsedWithInstancedStaticMeshes=true
  bUsedWithNanite=true
  bUsedWithNiagaraMeshParticles=true
  blendMode=BLEND_Translucent
  opacityMaskClipValue=0.1
  pixelDepthOffsetMode=PDOM_Legacy
  previewMesh=/Game/Environments/Helios/FX/Hades/Meshes/S_FX_Hades_Skull_Niagara_01
  shadingModel=MSM_Unlit
  usedShadingModels=Unlit
 }
 nodes [
  Add(
   A=NamedRerouteUsage()
   B=SphereMask(
    A=ObjectPositionWS()
    B=Constant3Vector()
    Radius=ScalarParameter() {defaultValue=100, parameterName=Sphere_Radius}
    Hardness=ScalarParameter() {parameterName=Sphere_Hardness}
   )
  )
  Constant() {r=5}
  DynamicParameter() {parameterIndex=1, defaultValue=<rgba 5 1 1 1>, paramNames [DepthFade,Param2,Param3,Param4]}
  n1 = Constant() {r=1}
  n2 = Multiply(A=@n1, B=NamedRerouteUsage())
  n3 = VertexColor()
  n4 = Multiply(A=@n3.B, B=@n3.B)
  n5 = LinearInterpolate(Alpha=Multiply(A=@n4, B=@n4))
  MaterialFunctionCall(Emissive=Multiply(A=ParticleColor().RGB, B=LinearInterpolate(A=@n2, B=@n1, Alpha=@n5))) {materialFunction=&MF_FX_EmissiveAdjustments}
  Multiply(A=@n3.B)
  n6 = Power(Base=OneMinus(None=Fresnel() {exponent=2})) {constExponent=4}
  n7 = DepthFade(
   Opacity=Multiply(A=NamedRerouteUsage(), B=ParticleColor().A)
   FadeDistance=DynamicParameter() {parameterIndex=1, defaultValue=<rgba 5 1 1 1>, paramNames [DepthFade,Param2,Param3,Param4]}.DepthFade
  )
  n8 = ParticlePositionWS()
  n9 = ScalarParameter() {defaultValue=2000, parameterName="Fade_Start_Distance (S)"}
  n10 = ScalarParameter() {defaultValue=1000, parameterName="Fade_End_Distance (S)"}
  Multiply(
   A=LinearInterpolate(
    A=ScalarParameter() {defaultValue=5, parameterName=FinalOpacity}
    B=ScalarParameter() {defaultValue=5, parameterName=FinalOpacitySkull}
    Alpha=NamedRerouteUsage()
   )
   B=Multiply(
    A=Multiply(A=@n6, B=@n7)
    B=MaterialFunctionCall(Position=@n8, Fade_Start_Distance=@n9, Fade_End_Distance=@n10) {materialFunction=&MF_Distance_SyncedFade_FX}.Result
   )
  )
  NamedRerouteDeclaration(None=@n5) {name=SkullVertex, nodeColor=<rgba 0.8281 1 0 1>}
  n11 = TextureCoordinate()
  n12 = Constant2Vector() {g=0.5, r=0.5}
  n13 = ScalarParameter() {defaultValue=0.5, parameterName=CenterDot_Radius}
  n14 = ScalarParameter() {parameterName=CenterDot_Hardness}
  NamedRerouteDeclaration(None=Multiply(A=SphereMask(A=@n11, B=@n12, Radius=@n13, Hardness=@n14), B=ScalarParameter() {defaultValue=1, parameterName=CenterDot_OpacitytMult})) {name=Centerdot, nodeColor=<rgba 0 0.2734 1 1>}
  n15 = DynamicParameter() {defaultValue=<rgba 5 1 1 1>, paramNames [CurlAmountRandom, CurlSpeed, CurlIntensity, CurlAmountFront]}
  n16 = DynamicParameter() {defaultValue=<rgba 5 1 1 1>, paramNames [CurlAmountRandom, CurlSpeed, CurlIntensity, CurlAmountFront]}
  n17 = Subtract(A=Subtract(A=WorldPosition().XYZ, B=ParticlePositionWS()), B=Multiply(A=Time(), B=@n16.CurlSpeed))
  n18 = Multiply(A=@n15.CurlIntensity, B=VectorNoise(World Position=Divide(A=@n17, B=@n16.CurlAmountRandom)) {noiseFunction=VNF_VectorALU})
  n19 = TextureCoordinate()
  n20 = Constant2Vector() {g=0.5, r=0.5}
  n21 = ScalarParameter() {defaultValue=1, parameterName=HArdd}
  NamedRerouteDeclaration(None=LinearInterpolate(
    A=@n18
    B=Multiply(A=@n18, B=@n15.CurlAmountFront)
    Alpha=SphereMask(A=@n19, B=@n20, Hardness=@n21) {attenuationRadius=0.4, hardnessPercent=0}
   )) {name=Curl, nodeColor=<rgba 1 0.6563 0 1>}
  n22 = Fresnel(ExponentIn=ScalarParameter() {defaultValue=1, parameterName=ExponentIn}) {baseReflectFraction=0.0001, exponent=0.1}
  NamedRerouteDeclaration(None=Power(Base=OneMinus(None=@n22))) {name=Fresnel, nodeColor=<rgba 0.0781 0 1 1>}
  n23 = MaterialFunctionCall() {materialFunction=&VectorToRadialValue}
  n24 = MaterialFunctionCall(Float2=@n23.Radial Coordinates) {materialFunction=&BreakOutFloat2Components}
  n25 = Multiply(
   A=Subtract(A=Multiply(A=@n24.G), B=ScalarParameter() {defaultValue=0.1, parameterName=MaskOffset})
   B=ScalarParameter() {defaultValue=0.2, parameterName=MaskScale}
  )
  n26 = DynamicParameter() {parameterIndex=2, defaultValue=<rgba -0.05 1 0.1 0.25>, paramNames [RandomPanSpeed, RandomNoiseTile, RandomNoise_Offset, RandomNoise_Pulse]}
  n27 = VectorParameter() {parameterName=Speed, defaultValue=<rgba -0.1 -0.1 1 1>}
  n28 = Add(A=@n26.RandomNoiseTile, B=AppendVector(A=@n27.B, B=@n27.B))
  n29 = AppendVector(A=@n27.R, B=@n27.R)
  n30 = TextureSample(UVs=Panner(
    Coordinate=MaterialFunctionCall(UVs=@n23.Radial Coordinates, Texture Scale=@n28) {materialFunction=&ScaleUVsByCenter}.UVs
    Speed=Add(A=@n26.RandomPanSpeed, B=@n29)
   )) {samplerType=SAMPLERTYPE_Normal, texture=/Game/VisualThreatManager/StormVisuals/LargeClouds/T_StormSwirlNormal}
  n31 = AppendVector(A=@n30.R, B=@n30.R)
  n32 = ScalarParameter() {defaultValue=1, parameterName=Noise_Distortion}
  n33 = Add(A=TextureCoordinate(), B=Multiply(A=Saturate(None=@n25), B=Multiply(A=@n31, B=@n32)))
  n34 = DynamicParameter() {parameterIndex=2, defaultValue=<rgba -0.05 1 0.1 0.25>, paramNames [RandomPanSpeed, RandomNoiseTile, RandomNoise_Offset, RandomNoise_Pulse]}
  n35 = TextureSample(UVs=Add(A=@n33, B=LinearInterpolate(Alpha=@n34.RandomNoise_Offset) {constB=0.2})) {texture=/Game/Effects/Fort_Effects/Textures/Tile/T_Noise_02}
  n36 = LinearInterpolate(Alpha=@n34.RandomNoise_Offset) {constB=0.2}
  n37 = Add(A=@n35.RGB, B=TextureSample(UVs=Add(A=@n33, B=@n36)) {texture=&T_Noise_02}.RGB)
  n38 = ComponentMask(None=@n37) {g=true, r=true}
  n39 = DynamicParameter() {parameterIndex=2, defaultValue=<rgba -0.05 1 0.1 0.25>, paramNames [RandomPanSpeed, RandomNoiseTile, RandomNoise_Offset, RandomNoise_Pulse]}
  n40 = LinearInterpolate(Alpha=MaterialFunctionCall(Speed=LinearInterpolate(Alpha=@n39.RandomNoise_Pulse) {constA=0.5}) {materialFunction=&DebugTimeSine}.Result) {constA=0.5}
  n41 = Multiply(
   A=@n37
   B=SphereMask(A=TextureCoordinate(), B=Constant2Vector() {g=0.5, r=0.5}, Radius=ScalarParameter() {defaultValue=0.48, parameterName=Rad}) {hardnessPercent=0}
  )
  n42 = Add(A=TextureSample(UVs=Multiply(A=@n38, B=@n40)) {texture=/Game/Effects/Fort_Effects/Textures/Tile/T_Caustics_Chromatic}.RGB, B=@n41)
  n43 = Constant()
  NamedRerouteDeclaration(None=Multiply(A=Desaturation(None=@n42, Fraction=@n43), B=ScalarParameter() {defaultValue=1, parameterName=Emissive})) {name=AnimatedTexture, nodeColor=<rgba 0.8516 1 0 1>}
  NamedRerouteDeclaration(None=@n41) {name=OpacityAnimatedTexture, nodeColor=<rgba 0.6641 1 0 1>}
  n44 = LinearInterpolate(A=ScalarParameter() {parameterName=HueShiftAmount}, B=ScalarParameter() {parameterName=HueShiftAmountMax}, Alpha=PerInstanceRandom())
  n45 = Multiply(A=ScalarParameter() {parameterName=HueShiftAmount}, B=Multiply(A=VertexColor().R, B=@n44)) {constB=-0.05}
  n46 = VectorParameter() {parameterName="Color_Day (V3)", defaultValue=<rgba 0 1 0.1012 1>}
  n47 = Fresnel() {exponent=0.8}
  NamedRerouteDeclaration(None=MaterialFunctionCall(
    Hue Shift Percentage=StaticSwitchParameter(True=@n45, False=@n44) {defaultValue=true, parameterName=UseVertexForHueShift?}
    Texture=Multiply(A=@n46.RGB, B=@n47)
   ) {materialFunction=&HueShift}.Result) {name=BaseColorTOD, nodeColor=<rgba 1 0.6563 0 1>}
  NamedRerouteUsage()
  ObjectPositionWS()
  OneMinus(None=Fresnel() {exponent=0.25})
  PerInstanceRandom()
  Power(Base=@n35.RGB)
  Saturate(None=ScalarParameter() {defaultValue=1, parameterName=SkullAnimation})
  ScalarParameter() {defaultValue=0.5}
  StaticSwitchParameter(
   True=MaterialFunctionCall() {materialFunction=/Game/Effects/Fort_Effects/Materials/_Functions/CurveColor_Time_of_Day}.Result
   False=MaterialFunctionCall(
    Color_Morning=VectorParameter() {parameterName="Color_Morning (V3)", defaultValue=<rgba 0 1 0.1012 1>}.RGB
    Color_Day=VectorParameter() {parameterName="Color_Day (V3)", defaultValue=<rgba 0 1 0.1012 1>}.RGB
    Color_Evening=VectorParameter() {parameterName="Color_Evening (V3)", defaultValue=<rgba 0 1 0.1012 1>}.RGB
    Color_Night=VectorParameter() {parameterName="Color_Night (V3)", defaultValue=<rgba 0 1 0.1012 1>}.RGB
   ) {materialFunction=&Color_Time_of_Day}.Result
  ) {parameterName=UseCurveforTODinsteadofScalar?}
  TextureSample(UVs=TextureCoordinate()) {texture=/Game/Packages/Fortress_Transylvania/Haunted_Forest/Swamps/Textures/2ndInfectedMapLinear2}
 ]
}
